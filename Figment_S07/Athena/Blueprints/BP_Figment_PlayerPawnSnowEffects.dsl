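; Player pawn component for Figment S07 that gives players snow-event gameplay effects, namely cold breath and a snow surface change.
; On BeginPlay, with authority only, it casts its owner to PlayerPawn_Athena and binds to that pawn's possessed event.
; On possession, it branches on the "sp" calendar event state to apply those effects through the pawn's Ability System Component.
; Tick is empty.
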(event EventBeginPlay
  (Networking|SwitchHasAuthority
    (:Authority
      (bind _asplayer_pawn_athena (Utilities|Casting|CastToPlayerPawn_Athena (Components|GetOwner)))
      (Game|Player|AssignOnPawnPossessed _asplayer_pawn_athena (delegate Custom|OnPawnPossessed_Event)))))

(event Custom|OnPawnPossessed_Event
  (params (Pawn :type Pawn%20Object%20Reference))
  (bind _returnvalue (Ability|GetAbilitySystemComponent Pawn))
  (bind _biseventactive (CalendarEvents|GetCalendarEventInformation "sp"))
  (if _biseventactive
    (GameplayEffects|ApplyGameplayEffectToSelf :self _returnvalue :GameplayEffectClass "/Figment_S07/Athena/Effects/Blueprints/ColdBreath/GE_Figment_ColdBreathActivate.GE_Figment_ColdBreathActivate_C" :Level 0.0)
    (GameplayEffects|ApplyGameplayEffectToSelf :self _returnvalue :GameplayEffectClass "/Figment_S07/Athena/Environments/Blueprints/SurfaceEffects/GE_Figment_SurfaceChange_Snow_Event.GE_Figment_SurfaceChange_Snow_Event_C" :Level 0.0)))

(event EventTick
  (params (DeltaSeconds :type Float%20%28single-precision%29)))
